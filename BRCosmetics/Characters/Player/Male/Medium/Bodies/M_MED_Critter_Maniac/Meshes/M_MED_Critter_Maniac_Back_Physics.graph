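# Physics asset for the back piece of the male medium Critter Maniac player cosmetic, covering its dangling back wrap and blade wrap bones.
# Capsules on the pelvis and spine form a torso body chain beneath the wraps.
# A four-bone back wrap chain and a two-bone blade wrap chain both hang from spine_04.
# Each wrap chain has a tighter twist limit at its root than along its length.
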PhysicsAsset /BRCosmetics/Characters/Player/Male/Medium/Bodies/M_MED_Critter_Maniac/Meshes/M_MED_Critter_Maniac_Back_Physics
bodies {
 dyn_backWrap_01 [{shapeType=Capsule, length=2, radius=2, shapeName=dyn_backWrap_04_capsule, center=<2.5 0 0>, rotation=<rot 90 0 0>}]
 dyn_backWrap_02 [{shapeType=Capsule, length=2, radius=2, shapeName=dyn_backWrap_04_capsule, center=<2.5 0 0>, rotation=<rot 90 0 0>}]
 dyn_backWrap_03 [{shapeType=Capsule, length=2, radius=2, shapeName=dyn_backWrap_04_capsule, center=<2.5 0 0>, rotation=<rot 90 0 0>}]
 dyn_backWrap_04 [{shapeType=Capsule, length=2, radius=3, shapeName=dyn_backWrap_04_capsule, center=<3.5 0 0>, rotation=<rot 90 0 0>}]
 dyn_bladeWrap_01 [{shapeType=Capsule, length=3, radius=2, shapeName=dyn_bladeWrap_02_capsule, center=<2.5 0 0>, rotation=<rot 90 0 0>}]
 dyn_bladeWrap_02 [{shapeType=Capsule, length=3, radius=2, shapeName=dyn_bladeWrap_02_capsule, center=<2.5 0 0>, rotation=<rot 90 0 0>}]
 pelvis [{shapeType=Capsule, length=7, radius=8.3, center=<-2 -2 0>}]
 spine_01 [{shapeType=Capsule, length=6, radius=7.8, shapeName=spine_01_capsule, center=<3 -0.25 0>}]
 spine_02 [{shapeType=Capsule, length=5, radius=8, center=<3 0 0>}]
 spine_03 [{shapeType=Capsule, length=6.5, radius=9.5, shapeName=spine_03_capsule, center=<3.5 0.5 0>}]
 spine_04 [{shapeType=Capsule, length=10, radius=10.8, center=<5 0.5 0>}]
}
constraints [
 {bone1Name=spine_02, bone2Name=spine_01, swing1LimitDegrees=45, swing2LimitDegrees=45, twistLimitDegrees=45}
 {bone1Name=spine_04, bone2Name=spine_03, swing1LimitDegrees=45, swing2LimitDegrees=45, twistLimitDegrees=45}
 {
  bone1Name=spine_01
  bone2Name=pelvis
  swing1LimitDegrees=45
  swing1Motion=Free
  swing2LimitDegrees=45
  swing2Motion=Free
  twistLimitDegrees=45
  twistMotion=Free
 }
 {
  bone1Name=spine_03
  bone2Name=spine_02
  swing1LimitDegrees=45
  swing1Motion=Free
  swing2LimitDegrees=45
  swing2Motion=Free
  twistLimitDegrees=45
  twistMotion=Free
 }
 {bone1Name=dyn_backWrap_01, bone2Name=spine_04, swing1LimitDegrees=45, swing2LimitDegrees=45, twistLimitDegrees=10}
 {bone1Name=dyn_backWrap_02, bone2Name=dyn_backWrap_01, swing1LimitDegrees=45, swing2LimitDegrees=45, twistLimitDegrees=30}
 {bone1Name=dyn_backWrap_03, bone2Name=dyn_backWrap_02, swing1LimitDegrees=45, swing2LimitDegrees=45, twistLimitDegrees=30}
 {bone1Name=dyn_backWrap_04, bone2Name=dyn_backWrap_03, swing1LimitDegrees=45, swing2LimitDegrees=45, twistLimitDegrees=30}
 {bone1Name=dyn_bladeWrap_01, bone2Name=spine_04, swing1LimitDegrees=45, swing2LimitDegrees=45, twistLimitDegrees=10}
 {bone1Name=dyn_bladeWrap_02, bone2Name=dyn_bladeWrap_01, swing1LimitDegrees=45, swing2LimitDegrees=45, twistLimitDegrees=30}
]
