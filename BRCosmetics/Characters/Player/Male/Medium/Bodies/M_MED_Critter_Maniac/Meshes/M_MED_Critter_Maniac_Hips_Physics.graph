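# Physics asset for the hips mesh of the Critter Maniac medium male body cosmetic, covering a dynamic glute wrap, pelvis and thighs.
# Every body is a single capsule, with large ones on pelvis and thighs and small ones along the four wrap bones.
# The wrap bones form a constraint chain from the pelvis, with twist much tighter at the root joint than further down.
# Both thighs constrain directly to the pelvis.
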PhysicsAsset /BRCosmetics/Characters/Player/Male/Medium/Bodies/M_MED_Critter_Maniac/Meshes/M_MED_Critter_Maniac_Hips_Physics
bodies {
 dyn_gluteWrap_01 [{shapeType=Capsule, length=3, radius=2.2, shapeName=dyn_gluteWrap_02_capsule, center=<3 0 0>, rotation=<rot 90 0 0>}]
 dyn_gluteWrap_02 [{shapeType=Capsule, length=3, radius=2.6, shapeName=dyn_gluteWrap_02_capsule, center=<3 0 0>, rotation=<rot 90 0 0>}]
 dyn_gluteWrap_03 [{shapeType=Capsule, length=3, radius=2, shapeName=dyn_gluteWrap_02_capsule, center=<3 0 0>, rotation=<rot 90 0 0>}]
 dyn_gluteWrap_04 [{shapeType=Capsule, length=5, radius=2, shapeName=dyn_gluteWrap_02_capsule, center=<4.5 0 0>, rotation=<rot 90 0 0>}]
 pelvis [{shapeType=Capsule, length=7, radius=8.3, center=<-2 -2 0>}]
 thigh_l [{shapeType=Capsule, length=26, radius=6.8, center=<-19.5 0 0>, rotation=<rot 90 0 6>}]
 thigh_r [{shapeType=Capsule, length=26, radius=6.8, center=<19.5 0 0>, rotation=<rot 90 0 6>}]
}
constraints [
 {bone1Name=thigh_l, bone2Name=pelvis, swing1LimitDegrees=45, swing2LimitDegrees=45, twistLimitDegrees=45}
 {bone1Name=thigh_r, bone2Name=pelvis, swing1LimitDegrees=45, swing2LimitDegrees=45, twistLimitDegrees=45}
 {bone1Name=dyn_gluteWrap_01, bone2Name=pelvis, swing1LimitDegrees=45, swing2LimitDegrees=45, twistLimitDegrees=5}
 {bone1Name=dyn_gluteWrap_02, bone2Name=dyn_gluteWrap_01, swing1LimitDegrees=45, swing2LimitDegrees=45, twistLimitDegrees=30}
 {bone1Name=dyn_gluteWrap_03, bone2Name=dyn_gluteWrap_02, swing1LimitDegrees=45, swing2LimitDegrees=45, twistLimitDegrees=30}
 {bone1Name=dyn_gluteWrap_04, bone2Name=dyn_gluteWrap_03, swing1LimitDegrees=45, swing2LimitDegrees=45, twistLimitDegrees=30}
]
