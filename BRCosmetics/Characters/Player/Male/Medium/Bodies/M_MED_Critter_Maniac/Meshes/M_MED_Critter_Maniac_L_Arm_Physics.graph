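# Physics asset for the left arm of the M_MED_Critter_Maniac male player cosmetic body, setting up a dynamic arm wrap.
# Capsule bodies cover the upper arm, forearm and three chained dyn_armWrap bones.
# The wrap chain hangs from the forearm with angular limits: wide swing and tight twist at the root, narrower swing and looser twist between segments.
# The forearm-to-upper-arm joint leaves swing and twist free.
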PhysicsAsset /BRCosmetics/Characters/Player/Male/Medium/Bodies/M_MED_Critter_Maniac/Meshes/M_MED_Critter_Maniac_L_Arm_Physics
bodies {
 dyn_armWrap_l_01 [{shapeType=Capsule, length=3, radius=1.8, shapeName=dyn_armWrap_l_02_capsule, center=<3 0 0>, rotation=<rot 90 0 0>}]
 dyn_armWrap_l_02 [{shapeType=Capsule, length=3, radius=1.8, shapeName=dyn_armWrap_l_02_capsule, center=<3 0 0>, rotation=<rot 90 0 0>}]
 dyn_armWrap_l_03 [{shapeType=Capsule, length=3, radius=1.8, shapeName=dyn_armWrap_l_02_capsule, center=<3 0 0>, rotation=<rot 90 0 0>}]
 lowerarm_l [{shapeType=Capsule, length=15, radius=2.2, shapeName=lowerarm_l_capsule, center=<10 0 0>, rotation=<rot 90 0 0>}]
 upperarm_l [{shapeType=Capsule, length=22, radius=3.5, shapeName=upperarm_l_capsule, center=<15 0 0>, rotation=<rot 90 0 0>}]
}
constraints [
 {bone1Name=dyn_armWrap_l_01, bone2Name=lowerarm_l, swing1LimitDegrees=90, swing2LimitDegrees=90, twistLimitDegrees=10}
 {bone1Name=dyn_armWrap_l_02, bone2Name=dyn_armWrap_l_01, swing1LimitDegrees=45, swing2LimitDegrees=45, twistLimitDegrees=30}
 {bone1Name=dyn_armWrap_l_03, bone2Name=dyn_armWrap_l_02, swing1LimitDegrees=45, swing2LimitDegrees=45, twistLimitDegrees=30}
 {
  bone1Name=lowerarm_l
  bone2Name=upperarm_l
  swing1LimitDegrees=45
  swing1Motion=Free
  swing2LimitDegrees=45
  swing2Motion=Free
  twistLimitDegrees=45
  twistMotion=Free
 }
]
