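# Physics asset for the leg mesh of the Critter Maniac male medium cosmetic body, covering the right lower leg and its dangling wraps.
# Capsules on the calf, calf twist and foot form the leg, with a wide flat root box whose top sits at ground level.
# Three-link calf and shin wrap chains hang from the calf twist and calf bodies, with tighter twist limits at each chain's first joint.
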PhysicsAsset /BRCosmetics/Characters/Player/Male/Medium/Bodies/M_MED_Critter_Maniac/Meshes/M_MED_Critter_Maniac_Leg_Physics
bodies {
 calf_r [{shapeType=Capsule, length=36, radius=2.5, center=<22.0021 0.991 -0.1604>, rotation=<rot 88.5391 -3.4427 0>}]
 calf_twist_02_r [{shapeType=Capsule, length=8, radius=3, shapeName=calf_twist_02_r_capsule, center=<-5 2 0.5>, rotation=<rot 90 0 0>}]
 dyn_calfWrap_01 [{shapeType=Capsule, length=5, radius=2, shapeName=dyn_calfWrap_01_capsule, center=<4.5 0 0>, rotation=<rot 90 0 0>}]
 dyn_calfWrap_02 [{shapeType=Capsule, length=5, radius=2, shapeName=dyn_calfWrap_01_capsule, center=<4.5 0 0>, rotation=<rot 90 0 0>}]
 dyn_calfWrap_03 [{shapeType=Capsule, length=5, radius=2, shapeName=dyn_calfWrap_01_capsule, center=<4.5 0 0>, rotation=<rot 90 0 0>}]
 dyn_shinWrap_01 [{shapeType=Capsule, length=3, radius=2, shapeName=dyn_shinWrap_01_capsule, center=<3 0 0>, rotation=<rot 90 0 0>}]
 dyn_shinWrap_02 [{shapeType=Capsule, length=3, radius=2, shapeName=dyn_shinWrap_01_capsule, center=<3 0 0>, rotation=<rot 90 0 0>}]
 dyn_shinWrap_03 [{shapeType=Capsule, length=3, radius=2, shapeName=dyn_shinWrap_01_capsule, center=<3 0 0>, rotation=<rot 90 0 0>}]
 foot_r [{shapeType=Capsule, length=16, radius=4, center=<5 -8 0>, rotation=<rot 0 10 90>}]
 root [{shapeType=Box, extentX=100, extentY=100, extentZ=5, shapeName=root_box, center=<0 0 -2.5>}]
}
constraints [
 {bone1Name=calf_r, bone2Name=root, swing1LimitDegrees=45, swing2LimitDegrees=45, twistLimitDegrees=45}
 {bone1Name=foot_r, bone2Name=calf_r, swing1LimitDegrees=45, swing2LimitDegrees=45, twistLimitDegrees=45}
 {
  bone1Name=calf_twist_02_r
  bone2Name=calf_r
  swing1LimitDegrees=45
  swing1Motion=Free
  swing2LimitDegrees=45
  swing2Motion=Free
  twistLimitDegrees=45
  twistMotion=Free
 }
 {bone1Name=dyn_calfWrap_01, bone2Name=calf_twist_02_r, swing1LimitDegrees=45, swing2LimitDegrees=45, twistLimitDegrees=10}
 {bone1Name=dyn_calfWrap_02, bone2Name=dyn_calfWrap_01, swing1LimitDegrees=45, swing2LimitDegrees=45, twistLimitDegrees=30}
 {bone1Name=dyn_calfWrap_03, bone2Name=dyn_calfWrap_02, swing1LimitDegrees=45, swing2LimitDegrees=45, twistLimitDegrees=30}
 {bone1Name=dyn_shinWrap_01, bone2Name=calf_r, swing1LimitDegrees=45, swing2LimitDegrees=45, twistLimitDegrees=10}
 {bone1Name=dyn_shinWrap_02, bone2Name=dyn_shinWrap_01, swing1LimitDegrees=45, swing2LimitDegrees=45, twistLimitDegrees=30}
 {bone1Name=dyn_shinWrap_03, bone2Name=dyn_shinWrap_02, swing1LimitDegrees=45, swing2LimitDegrees=45, twistLimitDegrees=30}
]
