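# Physics asset for the right arm of the Critter Maniac medium male body cosmetic, setting up a two-bone dangling arm wrap for simulated motion.
# Capsule bodies cover the upper arm, forearm and both wrap bones.
# The wrap bones chain off the forearm with swing and twist limit angles, tighter in swing at the second link, while the forearm-to-upper-arm joint leaves swing and twist free.
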PhysicsAsset /BRCosmetics/Characters/Player/Male/Medium/Bodies/M_MED_Critter_Maniac/Meshes/M_MED_Critter_Maniac_R_Arm_Physics
bodies {
 dyn_armWrap_r_01 [{shapeType=Capsule, length=3, radius=1.6, shapeName=dyn_armWrap_r_01_capsule, center=<3 0 0>, rotation=<rot 90 0 0>}]
 dyn_armWrap_r_02 [{shapeType=Capsule, length=3, radius=1.6, shapeName=dyn_armWrap_r_01_capsule, center=<3 0 0>, rotation=<rot 90 0 0>}]
 lowerarm_r [{shapeType=Capsule, length=15, radius=2.2, shapeName=lowerarm_l_capsule, center=<-10 0 0>, rotation=<rot 90 0 0>}]
 upperarm_r [{shapeType=Capsule, length=22, radius=3.5, shapeName=upperarm_r_capsule, center=<-15 0 0>, rotation=<rot 90 0 0>}]
}
constraints [
 {bone1Name=dyn_armWrap_r_01, bone2Name=lowerarm_r, swing1LimitDegrees=90, swing2LimitDegrees=90, twistLimitDegrees=30}
 {bone1Name=dyn_armWrap_r_02, bone2Name=dyn_armWrap_r_01, swing1LimitDegrees=45, swing2LimitDegrees=45, twistLimitDegrees=30}
 {
  bone1Name=lowerarm_r
  bone2Name=upperarm_r
  swing1LimitDegrees=45
  swing1Motion=Free
  swing2LimitDegrees=45
  swing2Motion=Free
  twistLimitDegrees=45
  twistMotion=Free
 }
]
